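# Sound cue for a diving board bounce impact, routed to the Fort_Gadgets sound class.
# A quality-level node gives the first two levels an equal-weight random pick between two impact waves, and the last two a single "opt" wave.
# Overridden attenuation limits audible range, and overridden concurrency prevents new instances once the limit is reached.
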SoundCue /Game/Sounds/Fort_Test_Sounds/World_Objects/Bouncing/DivingBoard_Bounce_Impact_Cue
root=QualityLevel_0
nodes [
 {id=QualityLevel_0, children [Random_0,Random_0,WavePlayer_6,WavePlayer_6]}
 {id=Random_0, children [WavePlayer_2,WavePlayer_5], properties {weights [1,1]}}
 {id=WavePlayer_2, properties {soundWaveAssetPtr=&/DivingBoard_Bounce_Impact_02}}
 {id=WavePlayer_5, properties {soundWaveAssetPtr=&/DivingBoard_Bounce_Impact_01}}
 {id=WavePlayer_6, properties {soundWaveAssetPtr=&/DivingBoard_Bounce_Impact_01_opt}}
]
properties {
 bOverrideAttenuation=true
 bOverrideConcurrency=true
 maxDistance=1500
 soundClassObject=&Fort_Gadgets
 volumeMultiplier=0.8
 attenuationOverrides {falloffDistance=1000, attenuationShapeExtents {x=500}}
 concurrencyOverrides {resolutionRule=PreventNew}
}
