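# Sound cue for a hydrant launch, filed under bouncing world objects in a test sounds folder.
# A mixer layers the Hydrant_Launch wave with an impact wobble wave at equal volume, each through its own modulator with a raised pitch ceiling for variation.
# It plays in the player foley sound class with overridden distance attenuation and a shared concurrency rule named for a two-voice limit that stops the oldest.
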SoundCue /Game/Sounds/Fort_Test_Sounds/World_Objects/Bouncing/Hydrant_Launch_Cue
root=Mixer_1
nodes [
 {id=Mixer_1, children [Modulator_0,Modulator_1], properties {inputVolume [1,1]}}
 {id=Modulator_0, children [WavePlayer_0], properties {pitchMax=1.3}}
 {id=Modulator_1, children [WavePlayer_1], properties {pitchMax=1.5}}
 {id=WavePlayer_0, properties {soundWaveAssetPtr=&/Hydrant_Launch}}
 {id=WavePlayer_1, properties {soundWaveAssetPtr=/Game/Sounds/Fort_Impact_Sounds/Demolition/Defaults/Fort_Impact_Wobble_02}}
]
properties {
 bOverrideAttenuation=true
 maxDistance=1500
 soundClassObject=&Fort_Player_Foley
 volumeMultiplier=0.8
 attenuationOverrides {falloffDistance=1000, attenuationShapeExtents {x=500}}
 concurrencyOverrides {resolutionRule=PreventNew}
 concurrencySet [/Game/Sounds/Concurrency_Rules/TwoMaximum_StopOldest_VolumeScale]
}
