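# Sound cue for jumping or bouncing on a lilypad prop, routed to the Athena ambient props sound class.
# A Random node picks one of three equally weighted lilypad jump waves and is excluded from branch culling.
# A Modulator above it holds volume fixed, while the cue itself applies a slight volume boost.
# Playback uses a lilypad-specific attenuation asset, with a concurrency override that prevents new instances once the limit is reached.
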SoundCue /Game/Sounds/Fort_Test_Sounds/World_Objects/Bouncing/Prop_Lilypad_Jump_Cue
root=Modulator_1
nodes [
 {id=Modulator_1, children [Random_0], properties {volumeMax=1, volumeMin=1}}
 {id=Random_0, children [WavePlayer_2,WavePlayer_5,WavePlayer_6], properties {bShouldExcludeFromBranchCulling=true, weights [1,1,1]}}
 {id=WavePlayer_2, properties {soundWaveAssetPtr=&/Prop_Lilypad_Jump_03}}
 {id=WavePlayer_5, properties {soundWaveAssetPtr=&/Prop_Lilypad_Jump_01}}
 {id=WavePlayer_6, properties {soundWaveAssetPtr=&/Prop_Lilypad_Jump_02}}
]
properties {
 attenuationSettings=&/ATT_Prop_Lilypad_Jump
 bOverrideConcurrency=true
 maxDistance=6250
 soundClassObject=&Fort_Ambient_Props_Athena
 volumeMultiplier=1.3
 attenuationOverrides {falloffDistance=1000, stereoSpread=500, attenuationShapeExtents {x=500}}
 concurrencyOverrides {resolutionRule=PreventNew}
}
