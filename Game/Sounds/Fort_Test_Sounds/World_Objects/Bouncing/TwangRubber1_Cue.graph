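# Sound cue for a rubber bounce twang, built from tire bounce impact waves and classed as player foley.
# A quality-level node routes its first two levels to a random pick among three impacts and the last two to a single impact.
# Each branch passes through a modulator that varies pitch but keeps volume fixed.
# The cue overrides attenuation for distance falloff and sets concurrency to prevent new instances at the limit.
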SoundCue /Game/Sounds/Fort_Test_Sounds/World_Objects/Bouncing/TwangRubber1_Cue
root=QualityLevel_0
nodes [
 {id=Modulator_0, children [Random_1], properties {pitchMax=1, volumeMax=1, volumeMin=1}}
 {id=Modulator_1, children [WavePlayer_0], properties {pitchMax=1, volumeMax=1, volumeMin=1}}
 {id=QualityLevel_0, children [Modulator_0,Modulator_0,Modulator_1,Modulator_1]}
 {id=Random_1, children [WavePlayer_1,WavePlayer_3,WavePlayer_4], properties {weights [1,1,1]}}
 {id=WavePlayer_0, properties {soundWaveAssetPtr=&/Tire_Bounce_Impact_01}}
 {id=WavePlayer_1, properties {soundWaveAssetPtr=&Tire_Bounce_Impact_01}}
 {id=WavePlayer_3, properties {soundWaveAssetPtr=&/Tire_Bounce_Impact_02}}
 {id=WavePlayer_4, properties {soundWaveAssetPtr=&/Tire_Bounce_Impact_03}}
]
properties {
 bOverrideAttenuation=true
 bOverrideConcurrency=true
 maxDistance=1500
 soundClassObject=&Fort_Player_Foley
 volumeMultiplier=1.3
 attenuationOverrides {falloffDistance=1000, attenuationShapeExtents {x=500}}
 concurrencyOverrides {resolutionRule=PreventNew}
}
